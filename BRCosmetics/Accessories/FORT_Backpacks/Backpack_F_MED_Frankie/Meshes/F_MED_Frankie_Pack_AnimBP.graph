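# Animation Blueprint for the Frankie backpack, posing its wings to follow the wearer's movement.
# A state machine picks base, jump/fall, targeting, crouch or nested skydive poses, most with additive lean blendspaces driven by acceleration.
# Time-based sine and cosine offsets then float individual wing bones like shards.
# That cached result gains speed-driven Run and Sprint additive poses, zeroed during vertical motion and bypassed while crouching or skydiving.
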AnimBlueprint /BRCosmetics/Accessories/FORT_Backpacks/Backpack_F_MED_Frankie/Meshes/F_MED_Frankie_Pack_AnimBP
target_skeleton=/Game/Accessories/FORT_Backpacks/Backpack_F_MED_Frankie/Meshes/F_MED_Frankie_Pack_Skeleton
graphs [
 AnimGraph {nodes [
   n1 = Anim_Backpack_Frankie_MovePose_ADD_PoseAsset1 : Run!(PoseWeight:float=1) -> (Pose:Pose Link Structure) {
    is=title AnimGraphNode_PoseByName
    poseAsset=/BRCosmetics/Accessories/FORT_Backpacks/Backpack_F_MED_Frankie/Animation/Anim_Backpack_Frankie_MovePose_ADD_PoseAsset1
    poseName=Run
   }
   n2 = Anim_Backpack_Frankie_MovePose_ADD_PoseAsset1 : Sprint!(PoseWeight:float=1) -> (Pose:Pose Link Structure) {
    is=title AnimGraphNode_PoseByName
    poseAsset=&Anim_Backpack_Frankie_MovePose_ADD_PoseAsset1
    poseName=Sprint
   }
   n3 = Get SprintAlpha_b() -> (SprintAlpha_b:double)
   n4 = Two Way Blend(
    A=Apply Additive(Base=Use cached pose 'PreSprintCache'(), Additive=@n2, Alpha=@n3) {alphaBoolBlend {blendOption=Linear}, alphaScaleBiasClamp {bInterpResult=true, interpSpeedDecreasing=5}}
    B=Apply Additive(Base=Use cached pose 'PreSprintCache'(), Additive=@n1, Alpha=Get RunAlpha_b() -> (RunAlpha_b:double)) {alphaBoolBlend {blendOption=Linear}, alphaScaleBiasClamp {bInterpResult=true, interpSpeedDecreasing=5}}
    Alpha=@n3
   ) {blendNode {alphaBoolBlend {blendOption=Linear}, alphaScaleBiasClamp {bInterpResult=true, bias=1, interpSpeedDecreasing=5, scale=-1}}}
   n5 = Use cached pose 'PreSprintCache'()
   n6 = Get bIsCrouching()
   Output Pose(Result=Blend Poses by bool(
     BlendPose_0=Use cached pose 'PreSprintCache'()
     BlendPose_1=Blend Poses by bool(BlendPose_0=@n5, BlendPose_1=@n4, BlendTime_0=0.3, BlendTime_1=0.3, bActiveValue=@n6) {blendType=Linear}
     BlendTime_0=0.3
     BlendTime_1=0.3
     bActiveValue=Get bIsSkyDiving()
    ) {blendType=Linear}) {layerGroup=None}
   AnimGraphNode_StateMachine_0 = wingPosition() {is=title AnimGraphNode_StateMachine}
   n7 = Get ShardFloatA() -> (ShardFloatA:Vector)
   n8 = "Transform (Modify) Bone - Bone: dyn_wing_01_l"(ComponentPose=Local To Component(LocalPose=@AnimGraphNode_StateMachine_0), Alpha=1, Translation=@n7, Rotation="0,0,0", Scale="1,1,1") {translationMode=BMM_Additive, translationSpace=BCS_BoneSpace, alphaBoolBlend {blendOption=Linear}, boneToModify {boneName=dyn_wing_01_l}}
   n9 = "Transform (Modify) Bone - Bone: dyn_wing_03_l"(ComponentPose=@n8, Alpha=1, Translation=@n7, Rotation="0,0,0", Scale="1,1,1") {translationMode=BMM_Additive, translationSpace=BCS_BoneSpace, alphaBoolBlend {blendOption=Linear}, boneToModify {boneName=dyn_wing_03_l}}
   n10 = "Transform (Modify) Bone - Bone: dyn_wing_03_l"(ComponentPose=@n9, Alpha=1, Translation=@n7, Rotation="0,0,0", Scale="1,1,1") {translationMode=BMM_Additive, translationSpace=BCS_BoneSpace, alphaBoolBlend {blendOption=Linear}, boneToModify {boneName=dyn_wing_03_l}}
   n11 = Get ShardFloatB() -> (ShardFloatB:Vector)
   n12 = "Transform (Modify) Bone - Bone: dyn_wing_02_l"(ComponentPose=@n10, Alpha=1, Translation=@n11, Rotation="0,0,0", Scale="1,1,1") {translationMode=BMM_Additive, translationSpace=BCS_BoneSpace, alphaBoolBlend {blendOption=Linear}, boneToModify {boneName=dyn_wing_02_l}}
   n13 = "Transform (Modify) Bone - Bone: dyn_wing_04_l"(ComponentPose=@n12, Alpha=1, Translation=@n11, Rotation="0,0,0", Scale="1,1,1") {translationMode=BMM_Additive, translationSpace=BCS_BoneSpace, alphaBoolBlend {blendOption=Linear}, boneToModify {boneName=dyn_wing_04_l}}
   n14 = "Transform (Modify) Bone - Bone: dyn_wing_02_r"(ComponentPose=@n13, Alpha=1, Translation=@n11, Rotation="0,0,0", Scale="1,1,1") {translationMode=BMM_Additive, translationSpace=BCS_BoneSpace, alphaBoolBlend {blendOption=Linear}, boneToModify {boneName=dyn_wing_02_r}}
   n15 = "Transform (Modify) Bone - Bone: dyn_wing_04_r"(ComponentPose=@n14, Alpha=1, Translation=@n11, Rotation="0,0,0", Scale="1,1,1") {translationMode=BMM_Additive, translationSpace=BCS_BoneSpace, alphaBoolBlend {blendOption=Linear}, boneToModify {boneName=dyn_wing_04_r}}
   n16 = Get ShardFloatC() -> (ShardFloatC:Vector)
   n17 = "Transform (Modify) Bone - Bone: dyn_wing_01_r"(ComponentPose=@n15, Alpha=1, Translation=@n16, Rotation="0,0,0", Scale="1,1,1") {translationMode=BMM_Additive, translationSpace=BCS_BoneSpace, alphaBoolBlend {blendOption=Linear}, boneToModify {boneName=dyn_wing_01_r}}
   n18 = "Transform (Modify) Bone - Bone: dyn_wing_03_r"(ComponentPose=@n17, Alpha=1, Translation=@n16, Rotation="0,0,0", Scale="1,1,1") {translationMode=BMM_Additive, translationSpace=BCS_BoneSpace, alphaBoolBlend {blendOption=Linear}, boneToModify {boneName=dyn_wing_03_r}}
   n19 = "Transform (Modify) Bone - Bone: dyn_wing_05_r"(ComponentPose=@n18, Alpha=1, Translation=@n16, Rotation="0,0,0", Scale="1,1,1") {translationMode=BMM_Additive, translationSpace=BCS_BoneSpace, alphaBoolBlend {blendOption=Linear}, boneToModify {boneName=dyn_wing_05_r}}
   Save cached pose 'PreSprintCache'(Pose=Component To Local(ComponentPose=@n19)) {cacheName=PreSprintCache}
  ]}
 wingPosition {
  entry=Base
  path=AnimGraph.AnimGraphNode_StateMachine_0.wingPosition
  ids=id
  nodes [
   AnimStateNode_0 = Base() {is=title AnimStateNode}
   AnimStateNode_1 = Jumping() {is=title AnimStateNode}
   AnimStateNode_2 = Falling() {is=title AnimStateNode}
   AnimStateNode_3 = Targeting() {is=title AnimStateNode}
   AnimStateNode_5 = SkydiveMode() {is=title AnimStateNode}
   AnimStateNode_9 = Crouching() {is=title AnimStateNode}
   AnimStateTransitionNode_0 = AnimStateTransitionNode() {blendMode=Custom, customBlendCurve=/Game/Accessories/FORT_Backpacks/M_MED_Demon_Pack/Animation/JumpTransitionBlendCurve, priorityOrder=2}
   AnimStateTransitionNode_1 = AnimStateTransitionNode() {
    blendMode=Custom
    crossfadeDuration=0.5
    customBlendCurve=/Game/Accessories/FORT_Backpacks/M_MED_Demon_Pack/Animation/JumpTransitionBlendOutCurve
   }
   AnimStateTransitionNode_11 = AnimStateTransitionNode()
   AnimStateTransitionNode_12 = AnimStateTransitionNode()
   AnimStateTransitionNode_13 = AnimStateTransitionNode()
   AnimStateTransitionNode_15 = AnimStateTransitionNode()
   AnimStateTransitionNode_2 = AnimStateTransitionNode()
   AnimStateTransitionNode_3 = AnimStateTransitionNode() {blendMode=Custom, customBlendCurve=&JumpTransitionBlendOutCurve}
   AnimStateTransitionNode_4 = AnimStateTransitionNode() {blendMode=Custom, customBlendCurve=&JumpTransitionBlendCurve}
   AnimStateTransitionNode_5 = AnimStateTransitionNode()
   AnimStateTransitionNode_6 = AnimStateTransitionNode() {crossfadeDuration=0.25}
   AnimStateTransitionNode_7 = AnimStateTransitionNode() {crossfadeDuration=0.25}
  ]
  links [
   {
    dst=Crouching
    rule=AnimGraph.AnimGraphNode_StateMachine_0.wingPosition.AnimStateTransitionNode_12.AnimationTransitionGraph_0
    src=Base
    via=AnimStateTransitionNode_12
   }
   {
    dst=Falling
    rule=AnimGraph.AnimGraphNode_StateMachine_0.wingPosition.AnimStateTransitionNode_4.AnimationTransitionGraph_0
    src=Base
    via=AnimStateTransitionNode_4
   }
   {
    dst=Jumping
    rule=AnimGraph.AnimGraphNode_StateMachine_0.wingPosition.AnimStateTransitionNode_0.AnimationTransitionGraph_0
    src=Base
    via=AnimStateTransitionNode_0
   }
   {
    dst=SkydiveMode
    rule=AnimGraph.AnimGraphNode_StateMachine_0.wingPosition.AnimStateTransitionNode_15.AnimationTransitionGraph_0
    src=Base
    via=AnimStateTransitionNode_15
   }
   {
    dst=Targeting
    rule=AnimGraph.AnimGraphNode_StateMachine_0.wingPosition.AnimStateTransitionNode_6.AnimationTransitionGraph_0
    src=Base
    via=AnimStateTransitionNode_6
   }
   {
    dst=Base
    rule=AnimGraph.AnimGraphNode_StateMachine_0.wingPosition.AnimStateTransitionNode_13.AnimationTransitionGraph_0
    src=Crouching
    via=AnimStateTransitionNode_13
   }
   {
    dst=Base
    rule=AnimGraph.AnimGraphNode_StateMachine_0.wingPosition.AnimStateTransitionNode_3.AnimationTransitionGraph_0
    src=Falling
    via=AnimStateTransitionNode_3
   }
   {
    dst=Base
    rule=AnimGraph.AnimGraphNode_StateMachine_0.wingPosition.AnimStateTransitionNode_1.AnimationTransitionGraph_0
    src=Jumping
    via=AnimStateTransitionNode_1
   }
   {
    dst=Falling
    rule=AnimGraph.AnimGraphNode_StateMachine_0.wingPosition.AnimStateTransitionNode_2.AnimationTransitionGraph_0
    src=Jumping
    via=AnimStateTransitionNode_2
   }
   {
    dst=SkydiveMode
    rule=AnimGraph.AnimGraphNode_StateMachine_0.wingPosition.AnimStateTransitionNode_5.AnimationTransitionGraph_0
    src=Jumping
    via=AnimStateTransitionNode_5
   }
   {
    dst=Base
    rule=AnimGraph.AnimGraphNode_StateMachine_0.wingPosition.AnimStateTransitionNode_11.AnimationTransitionGraph_0
    src=SkydiveMode
    via=AnimStateTransitionNode_11
   }
   {
    dst=Base
    rule=AnimGraph.AnimGraphNode_StateMachine_0.wingPosition.AnimStateTransitionNode_7.AnimationTransitionGraph_0
    src=Targeting
    via=AnimStateTransitionNode_7
   }
  ]
 }
 Base {
  path=AnimGraph.AnimGraphNode_StateMachine_0.wingPosition.AnimStateNode_0.Base
  nodes [
   n1 = Get WingsBlendspaceFwdBwd_b() -> (WingsBlendspaceFwdBwd_b:double)
   n2 = Get WingsBlendspaceLeftRight_b() -> (WingsBlendspaceLeftRight_b:double)
   n3 = Anim_Backpack_Frankie_MovePose_PoseAsset : Base!(PoseWeight:float=1) -> (Pose:Pose Link Structure) {
    is=title AnimGraphNode_PoseByName
    poseAsset=/BRCosmetics/Accessories/FORT_Backpacks/Backpack_F_MED_Frankie/Animation/Anim_Backpack_Frankie_MovePose_PoseAsset
    poseName=Base
   }
   Output Animation Pose(Result=Apply Additive(
     Base=@n3
     Additive=Blendspace Player 'BS_Backpack_Frankie_Wing_Leans_Ground'(X=@n2, Y=@n1) {blendSpace=/BRCosmetics/Accessories/FORT_Backpacks/Backpack_F_MED_Frankie/Animation/BS_Backpack_Frankie_Wing_Leans_Ground}
    ) {alphaBoolBlend {blendOption=Linear}})
  ]
 }
 Jumping {
  path=AnimGraph.AnimGraphNode_StateMachine_0.wingPosition.AnimStateNode_1.Jumping
  nodes [
   n1 = Get WingsBlendspaceLeftRight_b()
   n2 = Get WingsBlendspaceFwdBwd_b()
   n3 = Anim_Backpack_Frankie_MovePose_PoseAsset : Jump!(PoseWeight:float=1) -> (Pose:Pose Link Structure) {
    is=title AnimGraphNode_PoseByName
    poseAsset=&Anim_Backpack_Frankie_MovePose_PoseAsset
    poseName=Jump
   }
   Output Animation Pose(Result=Apply Additive(
     Base=@n3
     Additive=Blendspace Player 'BS_Backpack_Frankie_Wing_Leans_Ground'(X=@n1, Y=@n2) {blendSpace=&BS_Backpack_Frankie_Wing_Leans_Ground}
    ) {alphaBoolBlend {blendOption=Linear}})
  ]
 }
 Falling {
  path=AnimGraph.AnimGraphNode_StateMachine_0.wingPosition.AnimStateNode_2.Falling
  nodes [
   n1 = Get WingsBlendspaceFwdBwd_b()
   n2 = Get WingsBlendspaceLeftRight_b()
   n3 = Anim_Backpack_Frankie_MovePose_PoseAsset : Jump(PoseWeight=1) {poseAsset=&Anim_Backpack_Frankie_MovePose_PoseAsset, poseName=Jump}
   Output Animation Pose(Result=Apply Additive(
     Base=@n3
     Additive=Blendspace Player 'BS_Backpack_Frankie_Wing_Leans_Ground'(X=@n2, Y=@n1) {blendSpace=&BS_Backpack_Frankie_Wing_Leans_Ground}
    ) {alphaBoolBlend {blendOption=Linear}})
  ]
 }
 Targeting {
  path=AnimGraph.AnimGraphNode_StateMachine_0.wingPosition.AnimStateNode_3.Targeting
  nodes [
   n1 = Anim_Backpack_Frankie_MovePose_PoseAsset : Targeting!(PoseWeight:float=1) -> (Pose:Pose Link Structure) {
    is=title AnimGraphNode_PoseByName
    poseAsset=&Anim_Backpack_Frankie_MovePose_PoseAsset
    poseName=Targeting
   }
   Output Animation Pose(Result=@n1)
  ]
 }
 SkydiveMode {
  path=AnimGraph.AnimGraphNode_StateMachine_0.wingPosition.AnimStateNode_5.SkydiveMode
  nodes [
   AnimGraphNode_StateMachine_0 = Skydiving State() {is=title AnimGraphNode_StateMachine}
   Output Animation Pose(Result=@AnimGraphNode_StateMachine_0)
  ]
 }
 Skydiving State {
  entry=Parachuting
  path=AnimGraph.AnimGraphNode_StateMachine_0.wingPosition.AnimStateNode_5.SkydiveMode.AnimGraphNode_StateMachine_0.Skydiving State
  ids=id
  nodes [
   AnimStateNode_1 = Diving() {is=title AnimStateNode}
   AnimStateNode_2 = Floating() {is=title AnimStateNode}
   AnimStateNode_4 = Parachuting() {is=title AnimStateNode}
   AnimStateTransitionNode_1 = AnimStateTransitionNode() {crossfadeDuration=0.5}
   AnimStateTransitionNode_10 = AnimStateTransitionNode()
   AnimStateTransitionNode_12 = AnimStateTransitionNode()
   AnimStateTransitionNode_13 = AnimStateTransitionNode()
   AnimStateTransitionNode_3 = AnimStateTransitionNode()
   AnimStateTransitionNode_5 = AnimStateTransitionNode() {crossfadeDuration=0.5}
  ]
  links [
   {
    dst=Floating
    rule=AnimGraph.AnimGraphNode_StateMachine_0.wingPosition.AnimStateNode_5.SkydiveMode.AnimGraphNode_StateMachine_0.Skydiving State.AnimStateTransitionNode_1.AnimationTransitionGraph_0
    src=Diving
    via=AnimStateTransitionNode_1
   }
   {
    dst=Parachuting
    rule=AnimGraph.AnimGraphNode_StateMachine_0.wingPosition.AnimStateNode_5.SkydiveMode.AnimGraphNode_StateMachine_0.Skydiving State.AnimStateTransitionNode_10.AnimationTransitionGraph_1
    src=Diving
    via=AnimStateTransitionNode_10
   }
   {
    dst=Diving
    rule=AnimGraph.AnimGraphNode_StateMachine_0.wingPosition.AnimStateNode_5.SkydiveMode.AnimGraphNode_StateMachine_0.Skydiving State.AnimStateTransitionNode_5.AnimationTransitionGraph_0
    src=Floating
    via=AnimStateTransitionNode_5
   }
   {
    dst=Parachuting
    rule=AnimGraph.AnimGraphNode_StateMachine_0.wingPosition.AnimStateNode_5.SkydiveMode.AnimGraphNode_StateMachine_0.Skydiving State.AnimStateTransitionNode_3.AnimationTransitionGraph_1
    src=Floating
    via=AnimStateTransitionNode_3
   }
   {
    dst=Diving
    rule=AnimGraph.AnimGraphNode_StateMachine_0.wingPosition.AnimStateNode_5.SkydiveMode.AnimGraphNode_StateMachine_0.Skydiving State.AnimStateTransitionNode_13.AnimationTransitionGraph_1
    src=Parachuting
    via=AnimStateTransitionNode_13
   }
   {
    dst=Floating
    rule=AnimGraph.AnimGraphNode_StateMachine_0.wingPosition.AnimStateNode_5.SkydiveMode.AnimGraphNode_StateMachine_0.Skydiving State.AnimStateTransitionNode_12.AnimationTransitionGraph_1
    src=Parachuting
    via=AnimStateTransitionNode_12
   }
  ]
 }
 Diving {
  path=AnimGraph.AnimGraphNode_StateMachine_0.wingPosition.AnimStateNode_5.SkydiveMode.AnimGraphNode_StateMachine_0.Skydiving State.AnimStateNode_1.Diving
  nodes [
   n1 = Get WingsBlendspaceFwdBwd_b()
   n2 = Get WingsBlendspaceLeftRight_b()
   n3 = Anim_Backpack_Frankie_MovePose_PoseAsset : Dive!(PoseWeight:float=1) -> (Pose:Pose Link Structure) {
    is=title AnimGraphNode_PoseByName
    poseAsset=&Anim_Backpack_Frankie_MovePose_PoseAsset
    poseName=Dive
   }
   Output Animation Pose(Result=Apply Additive(
     Base=@n3
     Additive=Blendspace Player 'BS_Backpack_Frankie_Wing_Leans_Diving'(X=@n2, Y=@n1) {blendSpace=/BRCosmetics/Accessories/FORT_Backpacks/Backpack_F_MED_Frankie/Animation/BS_Backpack_Frankie_Wing_Leans_Diving}
    ) {alphaBoolBlend {blendOption=Linear}})
  ]
 }
 Floating {
  path=AnimGraph.AnimGraphNode_StateMachine_0.wingPosition.AnimStateNode_5.SkydiveMode.AnimGraphNode_StateMachine_0.Skydiving State.AnimStateNode_2.Floating
  nodes [
   n1 = Get WingsBlendspaceFwdBwd_b()
   n2 = Get WingsBlendspaceLeftRight_b()
   n3 = Anim_Backpack_Frankie_MovePose_PoseAsset : Float!(PoseWeight:float=1) -> (Pose:Pose Link Structure) {
    is=title AnimGraphNode_PoseByName
    poseAsset=&Anim_Backpack_Frankie_MovePose_PoseAsset
    poseName=Float
   }
   Output Animation Pose(Result=Apply Additive(
     Base=@n3
     Additive=Blendspace Player 'BS_Backpack_Frankie_Wing_Leans_Floating'(X=@n2, Y=@n1) {blendSpace=/BRCosmetics/Accessories/FORT_Backpacks/Backpack_F_MED_Frankie/Animation/BS_Backpack_Frankie_Wing_Leans_Floating}
    ) {alphaBoolBlend {blendOption=Linear}})
  ]
 }
 Parachuting {
  path=AnimGraph.AnimGraphNode_StateMachine_0.wingPosition.AnimStateNode_5.SkydiveMode.AnimGraphNode_StateMachine_0.Skydiving State.AnimStateNode_4.Parachuting
  nodes [
   n1 = Get WingsBlendspaceFwdBwd_b()
   n2 = Get WingsBlendspaceLeftRight_b()
   n3 = Anim_Backpack_Frankie_MovePose_PoseAsset : Float(PoseWeight=1) {poseAsset=&Anim_Backpack_Frankie_MovePose_PoseAsset, poseName=Float}
   Output Animation Pose(Result=Apply Additive(
     Base=@n3
     Additive=Blendspace Player 'BS_Backpack_Frankie_Wing_Leans_Parachuting'(X=@n2, Y=@n1) {blendSpace=/BRCosmetics/Accessories/FORT_Backpacks/Backpack_F_MED_Frankie/Animation/BS_Backpack_Frankie_Wing_Leans_Parachuting}
    ) {alphaBoolBlend {blendOption=Linear}})
  ]
 }
 AnimationTransitionGraph_0 {
  path=AnimGraph.AnimGraphNode_StateMachine_0.wingPosition.AnimStateNode_5.SkydiveMode.AnimGraphNode_StateMachine_0.Skydiving State.AnimStateTransitionNode_1.AnimationTransitionGraph_0
  nodes [Result(bCanEnterTransition=NOT Boolean(A=Get bIsSkydiveDiveMode()))]
 }
 AnimationTransitionGraph_1 {
  path=AnimGraph.AnimGraphNode_StateMachine_0.wingPosition.AnimStateNode_5.SkydiveMode.AnimGraphNode_StateMachine_0.Skydiving State.AnimStateTransitionNode_10.AnimationTransitionGraph_1
  nodes [Result(bCanEnterTransition=Get bIsParachuteOpened())]
 }
 AnimationTransitionGraph_1 {
  path=AnimGraph.AnimGraphNode_StateMachine_0.wingPosition.AnimStateNode_5.SkydiveMode.AnimGraphNode_StateMachine_0.Skydiving State.AnimStateTransitionNode_12.AnimationTransitionGraph_1
  nodes [Result(bCanEnterTransition=NOT Boolean(A=Get bIsParachuteOpened()))]
 }
 AnimationTransitionGraph_1 {
  path=AnimGraph.AnimGraphNode_StateMachine_0.wingPosition.AnimStateNode_5.SkydiveMode.AnimGraphNode_StateMachine_0.Skydiving State.AnimStateTransitionNode_13.AnimationTransitionGraph_1
  nodes [Result(bCanEnterTransition=NOT Boolean(A=Get bIsParachuteOpened()))]
 }
 AnimationTransitionGraph_1 {
  path=AnimGraph.AnimGraphNode_StateMachine_0.wingPosition.AnimStateNode_5.SkydiveMode.AnimGraphNode_StateMachine_0.Skydiving State.AnimStateTransitionNode_3.AnimationTransitionGraph_1
  nodes [Result(bCanEnterTransition=Get bIsParachuteOpened())]
 }
 AnimationTransitionGraph_0 {
  path=AnimGraph.AnimGraphNode_StateMachine_0.wingPosition.AnimStateNode_5.SkydiveMode.AnimGraphNode_StateMachine_0.Skydiving State.AnimStateTransitionNode_5.AnimationTransitionGraph_0
  nodes [Result(bCanEnterTransition=Get bIsSkydiveDiveMode())]
 }
 Crouching {
  path=AnimGraph.AnimGraphNode_StateMachine_0.wingPosition.AnimStateNode_9.Crouching
  nodes [
   n1 = Anim_Backpack_Frankie_MovePose_PoseAsset : Crouch!(PoseWeight:float=1) -> (Pose:Pose Link Structure) {
    is=title AnimGraphNode_PoseByName
    poseAsset=&Anim_Backpack_Frankie_MovePose_PoseAsset
    poseName=Crouch
   }
   Output Animation Pose(Result=@n1)
  ]
 }
 AnimationTransitionGraph_0 {
  path=AnimGraph.AnimGraphNode_StateMachine_0.wingPosition.AnimStateTransitionNode_0.AnimationTransitionGraph_0
  nodes [Result(bCanEnterTransition=Get bIsJumping())]
 }
 AnimationTransitionGraph_0 {
  path=AnimGraph.AnimGraphNode_StateMachine_0.wingPosition.AnimStateTransitionNode_1.AnimationTransitionGraph_0
  nodes [
   Result(bCanEnterTransition=NOT Boolean(A=Get bIsJumping()))
   Get bShouldPredictLanding() -> (bShouldPredictLanding:bool)
   Get bLandingPredicted() -> (bLandingPredicted:bool)
   Get bDebugLandPrediction() -> (bDebugLandPrediction:bool)
   Get bIsFalling()
  ]
 }
 AnimationTransitionGraph_0 {
  path=AnimGraph.AnimGraphNode_StateMachine_0.wingPosition.AnimStateTransitionNode_11.AnimationTransitionGraph_0
  nodes [Result(bCanEnterTransition=NOT Boolean(A=Get bIsSkyDiving()))]
 }
 AnimationTransitionGraph_0 {
  path=AnimGraph.AnimGraphNode_StateMachine_0.wingPosition.AnimStateTransitionNode_12.AnimationTransitionGraph_0
  nodes [Result(bCanEnterTransition=Get bIsCrouching())]
 }
 AnimationTransitionGraph_0 {
  path=AnimGraph.AnimGraphNode_StateMachine_0.wingPosition.AnimStateTransitionNode_13.AnimationTransitionGraph_0
  nodes [Result(bCanEnterTransition=NOT Boolean(A=Get bIsCrouching()))]
 }
 AnimationTransitionGraph_0 {
  path=AnimGraph.AnimGraphNode_StateMachine_0.wingPosition.AnimStateTransitionNode_15.AnimationTransitionGraph_0
  nodes [Result(bCanEnterTransition=Get bIsSkyDiving())]
 }
 AnimationTransitionGraph_0 {
  path=AnimGraph.AnimGraphNode_StateMachine_0.wingPosition.AnimStateTransitionNode_2.AnimationTransitionGraph_0
  nodes [Result(bCanEnterTransition=Get bIsFalling())]
 }
 AnimationTransitionGraph_0 {
  path=AnimGraph.AnimGraphNode_StateMachine_0.wingPosition.AnimStateTransitionNode_3.AnimationTransitionGraph_0
  nodes [Result(bCanEnterTransition=NOT Boolean(A=Get bIsFalling()))]
 }
 AnimationTransitionGraph_0 {
  path=AnimGraph.AnimGraphNode_StateMachine_0.wingPosition.AnimStateTransitionNode_4.AnimationTransitionGraph_0
  nodes [Result(bCanEnterTransition=Get bIsFalling())]
 }
 AnimationTransitionGraph_0 {
  path=AnimGraph.AnimGraphNode_StateMachine_0.wingPosition.AnimStateTransitionNode_5.AnimationTransitionGraph_0
  nodes [Result(bCanEnterTransition=Get bIsSkyDiving())]
 }
 AnimationTransitionGraph_0 {
  path=AnimGraph.AnimGraphNode_StateMachine_0.wingPosition.AnimStateTransitionNode_6.AnimationTransitionGraph_0
  nodes [Result(bCanEnterTransition=Get bIsTargeting())]
 }
 AnimationTransitionGraph_0 {
  path=AnimGraph.AnimGraphNode_StateMachine_0.wingPosition.AnimStateTransitionNode_7.AnimationTransitionGraph_0
  nodes [Result(bCanEnterTransition=NOT Boolean(A=Get bIsTargeting()))]
 }
 EventGraph {nodes [
   Event BlueprintUpdateAnimation()
   > n1 = Sequence()
   Set PawnSpeedRemap(execute=@n1.then_1, PawnSpeedRemap:double=MapRangeClamped(Value=Get PawnSpeed2D(), InRangeA=0, InRangeB=549)) -> (Output_Get:double)
   > Sequence()
   n2 = GetGameTimeInSeconds()
   n3 = "Sin (Radians)"(A=float * float(A=float * float(A=1, B=1), B=@n2))
   n4 = "Sin (Radians)"(A=float * float(A=float * float(A=1, B=2), B=@n2))
   n5 = "Sin (Radians)"(A=float * float(A=float * float(A=1, B=1.5), B=@n2))
   > Set ShardFloatA(ShardFloatA:Vector=MakeVector(X=float * float(A=@n5, B=0.3), Y=float * float(A=@n3, B=0.56), Z=float * float(A=@n4, B=0.22))) -> (Output_Get:Vector)
   n6 = "Cos (Radians)"!(A:double=float * float(A=float * float(A=1, B=1.5), B=@n2)) -> (ReturnValue:double)
   n7 = "Cos (Radians)"(A=float * float(A=float * float(A=1, B=1.78), B=@n2))
   n8 = "Cos (Radians)"(A=float * float(A=float * float(A=1, B=1.15), B=@n2))
   > Set ShardFloatB(ShardFloatB:Vector=MakeVector(X=float * float(A=@n6, B=0.5), Y=float * float(A=@n7, B=0.3), Z=float * float(A=@n8, B=0.5))) -> (Output_Get:Vector)
   n9 = "Sin (Radians)"(A=float * float(A=float * float(A=1, B=1.5), B=@n2))
   n10 = "Sin (Radians)"(A=float * float(A=float * float(A=1, B=1), B=@n2))
   n11 = "Sin (Radians)"(A=float * float(A=float * float(A=1, B=2), B=@n2))
   > Set ShardFloatC(ShardFloatC:Vector=MakeVector(X=float * float(A=@n9, B=0.3), Y=float * float(A=@n10, B=0.6), Z=float * float(A=@n11, B=0.25))) -> (Output_Get:Vector)
   n12 = TryGetPawnOwner()
   n13 = GetVelocity!(self:Actor Object Reference=@n12) -> (ReturnValue:Vector)
   n14 = SelectFloat(A=0, B=1, bPickA="Not Equal (Float)"!(A:double=BreakVector(InVec=@n13).Z, B:double=0) -> (ReturnValue:bool) {is=title K2Node_CallFunction})
   n15 = Get Actor Rotation!(self:Actor Object Reference=@n12) -> (ReturnValue:Rotator) {is=title K2Node_CallFunction}
   n16 = BreakVector(InVec=vector * vector(A:Vector=Get Rotation X Vector!(InRot:Rotator=@n15) -> (ReturnValue:Vector) {is=title K2Node_CallFunction}, B:Vector=@n13) -> (ReturnValue:Vector) {is=title K2Node_CommutativeAssociativeBinaryOperator})
   n17 = float + float(A=@n16, B=@n16.Y) {is=title K2Node_CommutativeAssociativeBinaryOperator}
   Set SprintAlpha_b(execute=@n1.then_0, SprintAlpha_b:double=float * float(A=MapRangeClamped(Value=@n17, InRangeA=415, InRangeB=550), B=@n14)) -> (Output_Get:double)
   > Set RunAlpha_b(RunAlpha_b:double=float * float(A=MapRangeClamped(Value=@n17, InRangeA=0, InRangeB=300), B=@n14)) -> (Output_Get:double)
   n18 = TryGetPawnOwner()
   > n19 = Cast To Character!(execute, Object=@n18) -> (then, CastFailed, AsCharacter:Character Object Reference) {is=title K2Node_DynamicCast}
   n20 = Get Actor Rotation(self=@n18)
   n21 = GetCurrentAcceleration!(self:Character Movement Component Object Reference=Get CharacterMovement!(self:Character Object Reference=@n19) -> (CharacterMovement:Character Movement Component Object Reference)) -> (ReturnValue:Vector)
   n22 = BreakVector(InVec=vector * vector(A=Get Rotation X Vector(InRot=@n20), B=@n21))
   > Set WingsBlendspaceFwdBwd_b(WingsBlendspaceFwdBwd_b:double=float + float(A=@n22, B=@n22.Y)) -> (Output_Get:double)
   n23 = BreakVector(InVec=vector * vector(A=GetAxes!(A:Rotator=@n20) -> (X:Vector, Y:Vector, Z:Vector).Y, B=@n21))
   > Set WingsBlendspaceLeftRight_b(WingsBlendspaceLeftRight_b:double=float + float(A=@n23, B=@n23.Y)) -> (Output_Get:double)
   TryGetPawnOwner()
  ]}
]
variables [
 DynamicsAlphaBP {type=number}
 PawnSpeedRemap {type=number}
 RunAlpha_b {type=number}
 ShardFloatA {type=Vector, default=<0.1546 -0.3549 0.2157>}
 ShardFloatB {type=Vector, default=<-0.4285 -0.2884 0.0019>}
 ShardFloatC {type=Vector, default=<0.1546 -0.3802 0.2451>}
 SprintAlpha_b {type=number}
 WingsBlendspaceFwdBwd_b {type=number}
 WingsBlendspaceLeftRight_b {type=number}
]
